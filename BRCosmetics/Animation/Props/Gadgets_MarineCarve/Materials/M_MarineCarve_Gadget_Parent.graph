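# Parent material for the MarineCarve gadget prop, adding a blinking red REC indicator to the shared 2023 character material.
# A mirrored, adjustable LCD mask texture and a procedural LED dot combine under a rounded-sine blink to override emissive with red.
# Remaining emissive gets Fresnel-driven saturation and hue shift.
# Blue in the source emissive masks areas to black base color, zero specular and fixed roughness between the PreFX and PostFX functions.
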Material /BRCosmetics/Animation/Props/Gadgets_MarineCarve/Materials/M_MarineCarve_Gadget_Parent
graph {
 properties {
  bUseMaterialAttributes=true
  bUsedWithMeshDeformer=true
  bUsedWithMorphTargets=true
  bUsedWithSkeletalMesh=true
  pixelDepthOffsetMode=PDOM_Legacy
  shadingModel=MSM_FromMaterialExpression
  usedShadingModels=Default Lit | Subsurface | Clear Coat | Subsurface Profile | Cloth
 }
 nodes [
  n1 = MaterialFunctionCall() {materialFunction=&MF_CharacterParent_2023_PreFX}
  n2 = GetMaterialAttributes(None=@n1.Result) {attributeGetTypes [
    #1
    #2
    #3
    #4
   ]}
  n3 = Fresnel(ExponentIn=Constant() {r=1})
  n4 = MaterialFunctionCall(Hue Shift Percentage=Multiply(A=@n3) {constB=0.1}, Texture=Desaturation(None=@n2.Emissive Color, Fraction=Multiply(A=@n3) {constB=-10})) {materialFunction=&HueShift}
  n5 = Constant3Vector() {constant=<rgba 1 0 0 0>}
  n6 = Time()
  n7 = ScalarParameter() {defaultValue=1, parameterName=REC_BlinkSpeed}
  n8 = ConstantBiasScale(None=Sine(None=Multiply(A=@n6, B=@n7)))
  n9 = Add(A=TextureCoordinate(), B=AppendVector(A=ScalarParameter() {parameterName=LCD_UOffset}, B=ScalarParameter() {parameterName=LCD_VOffset}))
  n10 = ScalarParameter()
  n11 = Saturate(None=MaterialFunctionCall(
    UVs=MaterialFunctionCall(UVs=@n9, "Rotation Angle (0-1)"=@n10) {materialFunction=&CustomRotator}.Rotated Values
    Texture Scale=ScalarParameter() {defaultValue=1, parameterName=LCD_Size}
   ) {materialFunction=&ScaleUVsByCenter}.UVs)
  n12 = ComponentMask(None=@n11) {r=true}
  n13 = TextureSample(UVs=AppendVector(A=OneMinus(None=@n12), B=ComponentMask(None=@n11) {g=true})) {samplerType=SAMPLERTYPE_Masks, texture=/BRCosmetics/Animation/Props/Gadgets_MarineCarve/Textures/T_MarineCarve_Emote_Prop_EmissiveMask}
  n14 = ScalarParameter() {parameterName=LED_UOffset}
  n15 = ScalarParameter() {parameterName=LED_VOffset}
  n16 = MaterialFunctionCall(UVs=Add(A=TextureCoordinate(), B=AppendVector(A=@n14, B=@n15)), Texture Scale=ScalarParameter() {defaultValue=1, parameterName=LED_Size}) {materialFunction=&ScaleUVsByCenter}
  n17 = Constant() {r=0.5}
  n18 = Round(None=OneMinus(None=Distance(A=@n16.UVs, B=@n17)))
  n19 = Multiply(A=NamedRerouteDeclaration(None=Round(None=@n8)) {name=REC_Blink, nodeColor=<rgba 0.4766 1 0 1>}, B=Add(A=@n13.R, B=Saturate(None=@n18)))
  n20 = ComponentMask(None=Ceil(None=@n2.Emissive Color)) {b=true}
  n21 = SetMaterialAttributes!(
   MaterialAttributes=@n2.MaterialAttributes
   Emissive Color=LinearInterpolate(A=Multiply(A=@n4.Result) {constB=2}, B=Multiply(A=@n19, B=@n5), Alpha=@n19)
   Roughness=LinearInterpolate(A=NamedRerouteUsage(), Alpha=@n20) {constB=0.5}
   Base Color=LinearInterpolate(A=@n2.Base Color, Alpha=@n20) {constB=0}
   Specular=LinearInterpolate(A=NamedRerouteUsage(), Alpha=@n20) {constB=0}
  ) {attributeSetTypes [
    #1
    #2
    #3
    #4
   ]}
  MaterialFunctionCall(Material=@n21, M=@n1.M, Pre Skinned Local Position=@n1.M, Distance Blend=@n1.M, Fresnel=@n1.M) {materialFunction=&MF_CharacterParent_2023_PostFX}
  Multiply(A=@n13.R) {constB=2}
  NamedRerouteDeclaration(None=@n2.Specular) {name=InSpecular, nodeColor=<rgba 0 0.6719 1 1>}
  NamedRerouteDeclaration(None=@n2.Roughness) {name=InRough, nodeColor=<rgba 0.9922 1 0 1>}
 ]
}
